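# Settings category navigator entry widget: a collapsible tab button with a fading focus outline, built on FortSettingsListEntrySetting_Category.
# Parent events set the name and drive selected, deselected and reset visual states on Button_Category, while collapse and expand functions play its animation, optionally instantly.
# Non-touch hover animates outline and button, with a Gate muting the hover sound on gamepad.
# Gamepad focus resolves to the inherited button.
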WidgetBlueprint /FortniteCoreUI/Settings/CategoryNavigator/SettingsListEntry_Category
animations [OnHover {
  binding_count=2
  length_seconds=0.1
  object=&self:OnHover
  sequence {
   bindings [
    FocusOutline {
     id=#1
     possessed_class=Image
     tracks [{
       type=MovieSceneFloatTrack
       display_name=Render Opacity
       sections [{type=MovieSceneFloatSection, end_frame=12, start_frame=0, channels [None {type=MovieSceneScriptingFloatChannel, keys [[0,0],[12,1]]}]}]
      }]
    }
    Border {id=#2, possessed_class=Border, tracks []}
   ]
   display_rate {numerator=120}
  }
 }]
graphs [
 CollapseCategory {nodes [
   n1 = CollapseCategory!() -> (then, bIsInstant:bool) {is=title K2Node_FunctionEntry}
   > n2 = Cast To WBP_Settings_Button_Tab_Collapsible!(execute, Object=Get Button_Category() -> (Button_Category:Common Button Base Object Reference)) -> (then, CastFailed, AsWBP Settings Button Tab Collapsible:WBP Settings Button Tab Collapsible Object Reference) {is=title K2Node_DynamicCast}
   > PlayCollapseAnimation!(execute, self:WBP Settings Button Tab Collapsible Object Reference=@n2, ReverseAnimation:bool=false, InstantAnimation:bool=@n1) -> (then)
  ]}
 EventGraph {nodes [
   n1 = Event SetCategoryName!() -> (OutputDelegate, then, InCategoryName:"Text (by ref)") {is=title K2Node_Event}
   > n2 = Cast To WBP_UIKit_ButtonGeneric_Base!(execute, Object=Get Button_Category()) -> (then, CastFailed, AsWBP UIKit Button Generic Base:WBP UIKit Button Generic Base Object Reference) {is=title K2Node_DynamicCast}
   > SetText!(execute, self:WBP UIKit Button Generic Base Object Reference=@n2, Text:Text=@n1.InCategoryName) -> (then)
   Event BP_ResetCategory() {is=title K2Node_Event}
   > n3 = Cast To WBP_Settings_Button_Tab_Collapsible(Object=Get Button_Category())
   > RequestTransitionOnBlocks!(execute, self:UIKit Modular Button Object Reference=@n3, State:EUIKitBlockVisualState Enum=Unfocused) -> (then)
   n4 = Event BP_SetIsSelected!() -> (OutputDelegate, then, bIsSelected:bool) {is=title K2Node_Event}
   > n5 = Branch(Condition=@n4.bIsSelected)
   n6 = Get Button_Category()
   > n7 = Cast To WBP_Settings_Button_Tab_Collapsible(Object=@n6)
   > RequestTransitionOnBlocks(self=@n7, State=Selected)
   > PlaySelectedAnimation!(execute, self:WBP Settings Button Tab Collapsible Object Reference=@n7, IsSelected:bool=true) -> (then)
   n8 = Cast To WBP_Settings_Button_Tab_Collapsible(execute=@n5.else, Object=@n6)
   > RequestTransitionOnBlocks(self=@n8, State=DeselectedIdle)
   > PlaySelectedAnimation(self=@n8, IsSelected=false)
   Event OnMouseEnter()
   n9 = IsUsingTouch(PlayerController=GetOwningPlayer())
   > n10 = Branch(Condition=@n9)
   n11 = Get OnHover() -> (OnHover:Widget Animation Object Reference)
   PlayAnimationForward(execute=@n10.else, InAnimation=@n11)
   n12 = Get Button_Category()
   > n13 = Cast To WBP_Settings_Button_Tab_Collapsible(Object=@n12)
   > PlayHoveredAnimation!(execute, self:WBP Settings Button Tab Collapsible Object Reference=@n13, Hovered:bool=true) -> (then)
   > n14 = Branch(Condition=IsUsingGamepad!(self:Fort Global UIContext Object Reference=Get FortGlobalUIContext() -> (ReturnValue:Fort Global UIContext Object Reference) {is=title K2Node_GetSubsystem}) -> (ReturnValue:bool))
   > n15 = Sequence()
   Gate!(Enter:Exec=[@n16.then_1, @n15.then_1], Open:Exec=@n16.then_0, Close:Exec=@n15.then_0, Toggle:Exec, bStartClosed:bool=false) -> (Exit:Exec) {is=title K2Node_MacroInstance}
   > SpawnSound2D!(
    execute
    Sound:Sound Base Object Reference=Get HoverSound() -> (HoverSound:Sound Base Object Reference)
    VolumeMultiplier:float=1
    PitchMultiplier:float=1
    StartTime:float=0
    ConcurrencySettings:Sound Concurrency Object Reference
    bPersistAcrossLevelTransition:bool=false
    bAutoDestroy:bool=true
   ) -> (then, ReturnValue:Audio Component Object Reference)
   n16 = Sequence(execute=@n14.else)
   Event OnMouseLeave!() -> (OutputDelegate, then, MouseEvent:"Pointer Event Structure (by ref)") {is=title K2Node_Event}
   > n17 = Branch(Condition=@n9)
   PlayAnimationReverse(execute=@n17.else, InAnimation=@n11)
   > n18 = Cast To WBP_Settings_Button_Tab_Collapsible(Object=@n12)
   > PlayHoveredAnimation(self=@n18, Hovered=false)
  ]}
 ExpandCategory {nodes [
   n1 = ExpandCategory!() -> (then, bIsInstant:bool) {is=title K2Node_FunctionEntry}
   > n2 = Cast To WBP_Settings_Button_Tab_Collapsible(Object=Get Button_Category())
   > PlayCollapseAnimation(self=@n2, ReverseAnimation=true, InstantAnimation=@n1)
  ]}
 GetPrimaryGamepadFocusWidget {nodes [
   GetPrimaryGamepadFocusWidget() {is=title K2Node_FunctionEntry}
   > Return Node!(execute, ReturnValue:Widget Object Reference=Get Button_Category())
  ]}
]
variables [
 HoverSound {type=SoundBase, default=/Game/Sounds/Fort_UI_Sounds/FrontEnd/DefaultSounds/PressedButtons/MSS_UI_Button_Hover_Default}
 PressedSound {type=SoundBase, default=/Game/Sounds/Fort_UI_Sounds/FrontEnd/DefaultSounds/PressedButtons/MSS_UI_Button_Pressed_Default}
]
widgets {
 info {
  inheritedWidgetCount=1
  namedSlotCount=1
  parentClass=/Script/FortniteUI.FortSettingsListEntrySetting_Category
  rootWidgetClass=&Border
  widgetCount=5
 }
 widgets [
  Border {
   widgetClassPath=&Border
   bIsVariable=true
   widget=Border
   properties {visibility=SelfHitTestInvisible, brushColor=<rgba 1 1 1 0>, padding=<ltrb 10 8 10 8>}
  }
  Overlay_0 {widgetClassPath=&Overlay, parent=Border, widget=Overlay_0, slot {class=BorderSlot, properties {padding=<ltrb 10 8 10 8>}}}
  Button_Category {
   widgetClassPath=&/WBP_Settings_Button_Tab_Collapsible
   bInherited=true
   bIsVariable=true
   parent=Overlay_0
   widget=Button_Category
   properties {contentHorizontalAlignment=HAlign_Left, minHeight=56, wrapTextAt=300, buttonPadding=<ltrb 21 0 0 0>}
   slot {class=OverlaySlot, properties {horizontalAlignment=HAlign_Fill}}
  }
  Image_79 {
   widgetClassPath=&Image
   bIsVariable=true
   namedSlotHost=Button_Category
   widget=Image_79
   properties {brush {
     bIsDynamicallyLoaded=false
     drawAs=NoDrawType
     imageType=NoImage
     mirroring=NoMirror
     resourceName=None
     resourceObject=null
     tiling=NoTile
     imageSize=<32 32>
     margin=<ltrb 0 0 0 0>
     outlineSettings {
      bUseBrushTransparency=false
      roundingType=HalfHeightRadius
      width=0
      color {colorUseRule=UseColor_Specified, specifiedColor=<rgba 0 0 0 0>}
      cornerRadii=<0 0 0 0>
     }
     tintColor {colorUseRule=UseColor_Specified, specifiedColor=<rgba 1 1 1 1>}
     uVRegion {bIsValid=false, max=<0 0>, min=<0 0>}
    }}
  }
  FocusOutline {
   widgetClassPath=&Image
   bIsVariable=true
   parent=Overlay_0
   widget=FocusOutline
   properties {
    pixelSnapping=Disabled
    renderOpacity=0
    visibility=SelfHitTestInvisible
    brush {
     bIsDynamicallyLoaded=false
     drawAs=Image
     imageType=NoImage
     mirroring=NoMirror
     resourceName=None
     resourceObject=/FortniteCoreUI/Settings/Art/MI_UIKit_Settings_Outline
     tiling=NoTile
     imageSize=<32 32>
     margin=<ltrb 0 0 0 0>
     outlineSettings {
      bUseBrushTransparency=false
      roundingType=HalfHeightRadius
      width=0
      color {colorUseRule=UseColor_Specified, specifiedColor=<rgba 0 0 0 0>}
      cornerRadii=<0 0 0 0>
     }
     tintColor {colorUseRule=UseColor_Specified, specifiedColor=<rgba 1 1 1 1>}
     uVRegion {bIsValid=false, max=<0 0>, min=<0 0>}
    }
   }
   slot {class=OverlaySlot, properties {horizontalAlignment=HAlign_Fill, verticalAlignment=VAlign_Fill, padding=<ltrb -8 -16 -8 -16>}}
  }
 ]
}
